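# Physics asset for the TennisLeash large male player body cosmetic, covering a short dynamic chain and the upper spine.
# Capsule bodies sit on dyn_chain_base, dyn_chain_1 and dyn_chain_2, linked in sequence by two constraints with narrow swing and twist limit angles.
# Larger capsules on spine_04 and spine_05 have no constraints, which suggests they act as torso collision for the chain.
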PhysicsAsset /BRCosmetics/Characters/Player/Male/Large/Bodies/M_LRG_TennisLeash/Meshes/M_LRG_TennisLeash_Physics
bodies {
 dyn_chain_1 [{shapeType=Capsule, length=11.75, radius=5.5, shapeName=dyn_chain_1_capsule, center=<5.3688 0 0>, rotation=<rot -0.0002 0 90>}]
 dyn_chain_2 [{shapeType=Capsule, length=7.25, radius=2.25, shapeName=dyn_chain_2_capsule, center=<5.3688 0 0>, rotation=<rot -0.0002 0 90>}]
 dyn_chain_base [{shapeType=Capsule, length=0, radius=2, shapeName=dyn_chain_base_capsule}]
 spine_04 [{shapeType=Capsule, length=11.75, radius=15, shapeName=spine_04_capsule}]
 spine_05 [{shapeType=Capsule, length=17, radius=9.75, shapeName=spine_05_capsule, center=<-1.9943 0 0>}]
}
constraints [
 {bone1Name=dyn_chain_1, bone2Name=dyn_chain_base, swing1LimitDegrees=15, swing2LimitDegrees=45, twistLimitDegrees=10}
 {bone1Name=dyn_chain_2, bone2Name=dyn_chain_1, swing1LimitDegrees=20, swing2LimitDegrees=30, twistLimitDegrees=15}
]
