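# Physics asset for the face accessory mesh of the large male TennisLeash player cosmetic, covering dynamic beard, cap tail and hood bones.
# Every body is a single capsule, including the head, neck, upper spine and left clavicle.
# Angular-limited constraints chain the cap tail in two links, the beard to its base and the central hood to spine_05.
# Side beard and side/upper hood pieces have all angular motion locked to their parents.
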PhysicsAsset /BRCosmetics/Characters/Player/Male/Large/Bodies/M_LRG_TennisLeash/Meshes/Parts/M_LRG_TennisLeash_FaceAcc_Physics
bodies {
 clavicle_l [{
   shapeType=Capsule
   length=9.255
   radius=9.505
   shapeName=clavicle_l_capsule
   center=<11.8282 -0.315 1.715>
   rotation=<rot 1.0024 -89.6888 103.2409>
  }]
 dyn_beard [{shapeType=Capsule, length=16.5, radius=2.75, shapeName=dyn_beard_capsule, center=<9.3773 2.3035 0>}]
 dyn_beard_base [{shapeType=Capsule, length=0.505, radius=3.005, shapeName=dyn_beard_base_capsule, rotation=<rot 0 -90 0>}]
 dyn_beard_l [{shapeType=Capsule, length=9.5, radius=2.5, shapeName=dyn_beard_l_capsule, rotation=<rot 90 -123.0222 -123.0225>}]
 dyn_beard_r [{shapeType=Capsule, length=9.5, radius=2.5, shapeName=dyn_beard_r_capsule, rotation=<rot 90 -123.0222 -123.0225>}]
 dyn_cap_tail [{shapeType=Capsule, length=5.5, radius=2, shapeName=dyn_cap_tail_capsule, center=<6.5982 0 0>, rotation=<rot 79.9999 -179.9993 -179.9997>}]
 dyn_cap_tail_2 [{shapeType=Capsule, length=0, radius=3, shapeName=dyn_cap_tail_2_capsule, center=<2.2077 0 0>}]
 dyn_cap_tail_base [{shapeType=Capsule, length=0, radius=3.005, shapeName=dyn_cap_tail_base_capsule, rotation=<rot 0 -90 0>}]
 dyn_hood [{
   shapeType=Capsule
   length=12.2314
   radius=3.7207
   shapeName=dyn_hood_capsule
   center=<14.3583 -1.9381 -0.2903>
   rotation=<rot 1.1231 -90 3.5318>
  }]
 dyn_hood_l [{shapeType=Capsule, length=20, radius=7, shapeName=dyn_hood_l_capsule, center=<0.353 6.1295 -2.5819>, rotation=<rot -80 135.0016 -175>}]
 dyn_hood_r [{shapeType=Capsule, length=19.5, radius=4.75, shapeName=dyn_hood_r_capsule, center=<1.3263 -3.6437 3.1609>, rotation=<rot -85 -44.9994 0>}]
 dyn_hood_up [{shapeType=Capsule, length=26, radius=6, shapeName=dyn_hood_up_capsule, center=<0.1723 0.0788 0>}]
 head [{shapeType=Capsule, length=9, radius=7, shapeName=head_capsule, center=<4.6229 -0.1323 0>, rotation=<rot -90 180 164.9995>}]
 neck_02 [{shapeType=Capsule, length=5.25, radius=6.75, shapeName=neck_02_capsule, rotation=<rot 0 -90.0003 90>}]
 spine_05 [{shapeType=Capsule, length=16.5, radius=15, shapeName=spine_05_capsule, center=<0 2.5945 0>}]
}
constraints [
 {
  bone1Name=dyn_hood_up
  bone2Name=spine_05
  swing1LimitDegrees=45
  swing1Motion=Locked
  swing2LimitDegrees=45
  swing2Motion=Locked
  twistLimitDegrees=45
  twistMotion=Locked
 }
 {
  bone1Name=dyn_beard_r
  bone2Name=dyn_beard
  swing1LimitDegrees=45
  swing1Motion=Locked
  swing2LimitDegrees=45
  swing2Motion=Locked
  twistLimitDegrees=45
  twistMotion=Locked
 }
 {
  bone1Name=dyn_beard_l
  bone2Name=dyn_beard
  swing1LimitDegrees=45
  swing1Motion=Locked
  swing2LimitDegrees=45
  swing2Motion=Locked
  twistLimitDegrees=45
  twistMotion=Locked
 }
 {
  bone1Name=dyn_hood_l
  bone2Name=spine_05
  swing1LimitDegrees=45
  swing1Motion=Locked
  swing2LimitDegrees=45
  swing2Motion=Locked
  twistLimitDegrees=45
  twistMotion=Locked
 }
 {
  bone1Name=dyn_hood_r
  bone2Name=spine_05
  swing1LimitDegrees=45
  swing1Motion=Locked
  swing2LimitDegrees=45
  swing2Motion=Locked
  twistLimitDegrees=45
  twistMotion=Locked
 }
 {bone1Name=dyn_cap_tail_2, bone2Name=dyn_cap_tail, swing1LimitDegrees=60, swing2LimitDegrees=60, twistLimitDegrees=30}
 {bone1Name=dyn_hood, bone2Name=spine_05, swing1LimitDegrees=45, swing2LimitDegrees=45, twistLimitDegrees=45}
 {bone1Name=dyn_cap_tail, bone2Name=dyn_cap_tail_base, swing1LimitDegrees=15, swing2LimitDegrees=25, twistLimitDegrees=15}
 {bone1Name=dyn_beard, bone2Name=dyn_beard_base, swing1LimitDegrees=60, swing2LimitDegrees=20, twistLimitDegrees=60}
]
