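# Post-processes the face accessory mesh of the large male TennisLeash body, adding procedural and physics motion, notably to the beard.
# Copies the attached parent's pose into a Control Rig, offsets the beard base in the front end, then runs RigidBody simulation relative to spine_05.
# Update events turn crouch, skydive-float and front-end states into weights; customization load reads a material parameter to set hat and falloff values.
# Glasses constraint and offset nodes are unconnected.
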AnimBlueprint /BRCosmetics/Characters/Player/Male/Large/Bodies/M_LRG_TennisLeash/Meshes/Parts/M_LRG_TennisLeash_FaceAcc_AnimBP
target_skeleton=&/M_LRG_TennisLeash_FaceAcc_Skeleton
graphs [
 AnimGraph {nodes [
   "Constraint - dyn_glasses (T)"(Alpha=1, ConstraintWeights_0=1) {
    alphaBoolBlend {blendOption=Linear}
    boneToModify {boneName=dyn_glasses}
    constraintSetup [{offsetOption=None, transformType=Translation, targetBone {boneName=FACIAL_L_CheekOuter}}]
    constraintWeights [1]
   }
   n1 = Get Dyn_FrontEnd_Alpha() -> (Dyn_FrontEnd_Alpha:float)
   "Transform (Modify) Bone - Bone: dyn_glasses"(Alpha=@n1, Translation="0,1,0", Rotation="0,0,0", Scale="1,1,1") {
    rotationMode=BMM_Additive
    rotationSpace=BCS_BoneSpace
    translationMode=BMM_Additive
    translationSpace=BCS_BoneSpace
    alphaBoolBlend {blendOption=Linear}
    boneToModify {boneName=dyn_glasses}
    translation {y=1}
   }
   n2 = Control Rig!(
    Source=Copy Pose From Mesh() {bUseAttachedParent=true}
    Alpha
    is_crouch:double=Get Dyn_Is_Crouch() -> (Dyn_Is_Crouch:double)
    invert_crouch:double=Get Invert Crouch() -> (Invert Crouch:double)
    hat:double=Get Dyn_Hat() -> (Dyn_Hat:double)
    FalloffSize:float=Get Dyn_Falloff_Size() -> (Dyn_Falloff_Size:float)
    is_fly:double=Get Dyn_Skydive_Floating() -> (Dyn_Skydive_Floating:double)
   ) -> (Pose) {
    alphaBoolBlend {blendOption=Linear}
    alphaScaleBiasClamp {bInterpResult=true}
    controlRigAssetReference {blueprintRigClass=&/M_LRG_TennisLeash_FaceAcc_CtrlRig.M_LRG_TennisLeash_FaceAcc_CtrlRig_C}
   }
   n3 = "Transform (Modify) Bone - Bone: dyn_beard_base"(ComponentPose=Local To Component(LocalPose=@n2), Alpha=@n1, Translation="-1,2,0", Rotation="0.000,-30.000,-0.000", Scale="1,1,1") {
    rotationMode=BMM_Additive
    rotationSpace=BCS_BoneSpace
    translationMode=BMM_Additive
    translationSpace=BCS_BoneSpace
    alphaBoolBlend {blendOption=Linear}
    boneToModify {boneName=dyn_beard_base}
    rotation {yaw=-29.9999}
    translation=<-1 2>
   }
   n4 = RigidBody(ComponentPose=@n3) {
    bOverrideWorldGravity=true
    overridePhysicsAsset=&/M_LRG_TennisLeash_FaceAcc_Physics
    simulationSpace=BaseBoneSpace
    alphaBoolBlend {blendOption=Linear}
    alphaScaleBiasClamp {bInterpResult=true}
    baseBoneRef {boneName=spine_05}
    componentAppliedLinearAccClamp=<1250 1250 1250>
    componentLinearAccScale=<0.75 0.75 0.75>
    componentLinearVelScale=<0.5 0.5 0.5>
    overrideWorldGravity {z=-980}
    simSpaceSettings {maxAngularAcceleration=10, maxAngularVelocity=5, maxLinearAcceleration=1, maxLinearVelocity=1, velocityScaleZ=0.25, worldAlpha=1}
   }
   Output Pose(Result=Component To Local(ComponentPose=@n4)) {layerGroup=None}
  ]}
 EventGraph {nodes [
   Event BlueprintUpdateAnimation()
   > n1 = Sequence!(execute) -> (then_0, then_1, then_2:Exec)
   > n2 = Branch(Condition=Get bIsSkydiveFloating())
   > Set Dyn_Skydive_Floating(Dyn_Skydive_Floating:double=1) -> (Output_Get:double)
   Set Dyn_Skydive_Floating(execute=@n2.else, Dyn_Skydive_Floating=0)
   n3 = Branch(execute=@n1.then_1, Condition=Get bIsCrouching())
   > Set Dyn_Is_Crouch(Dyn_Is_Crouch:double=0.15) -> (Output_Get:double)
   > Set Invert Crouch(Invert Crouch:double=0) -> (Output_Get:double)
   Set Dyn_Is_Crouch(execute=@n3.else, Dyn_Is_Crouch=0)
   > Set Invert Crouch(Invert Crouch=1)
   n4 = Branch(execute=@n1.then_2, Condition=Get bIsInFrontEnd())
   > Set Dyn_FrontEnd_Alpha(Dyn_FrontEnd_Alpha:float=1) -> (Output_Get:float)
   Set Dyn_FrontEnd_Alpha(execute=@n4.else, Dyn_FrontEnd_Alpha=0)
   Event OnCharacterCustomizationLoaded()
   > n5 = Get Dynamic Material Instance(SkeletalMeshComponent=GetPartSkeletalMeshComponent())
   > n6 = Get Scalar Parameter Value!(execute, self:Material Instance Dynamic Object Reference=@n5, ParameterName:Name=Hide Element 03) -> (then, ReturnValue:float) {is=title K2Node_CallFunction}
   > n7 = Set Dyn_Hoog(Dyn_Hoog:double=@n6) -> (Output_Get:double)
   > n8 = Branch(Condition="To Bool (Double)"!(Value:double=@n7) -> (ReturnValue:bool))
   > Set Dyn_Hat(Dyn_Hat:double=1) -> (Output_Get:double)
   > Set Dyn_Falloff_Size(Dyn_Falloff_Size:float=0.28) -> (Output_Get:float)
   Set Dyn_Hat(execute=@n8.else, Dyn_Hat=0)
   > Set Dyn_Falloff_Size(Dyn_Falloff_Size=0.35)
  ]}
]
variables [
 Dyn_Falloff_Size {type=number}
 Dyn_FrontEnd_Alpha {type=number, default=1}
 Dyn_Hat {type=number}
 Dyn_Hoog {type=number}
 Dyn_Is_Crouch {type=number}
 Dyn_Skydive_Floating {type=number}
 Invert Crouch {type=number}
]
